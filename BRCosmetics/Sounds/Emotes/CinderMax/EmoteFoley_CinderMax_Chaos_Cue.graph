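# Emote foley Sound Cue for the CinderMax emote, building a randomized urban chaos soundscape in Fortnite.
# A continuously looping fire bed mixes with car alarms, honks, sirens and car crashes.
# Each one-shot layer loops through its own random delay and random wave pick, so events overlap irregularly, with one alarm variant weighted lower.
# It uses emote attenuation, foley concurrency limits and play-when-silent virtualization, and sends to a music audio-analysis submix.
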SoundCue /BRCosmetics/Sounds/Emotes/CinderMax/EmoteFoley_CinderMax_Chaos_Cue
root=Mixer_2
nodes [
 {id=Delay_0, children [Random_3], properties {delayMax=4, delayMin=0.5}}
 {id=Delay_2, children [Random_1], properties {delayMax=4, delayMin=1}}
 {id=Delay_3, children [Random_0], properties {delayMax=3}}
 {id=Delay_4, children [Random_4], properties {delayMax=6, delayMin=2}}
 {id=Looping_0, children [Mixer_1]}
 {id=Looping_1, children [Delay_0]}
 {id=Looping_3, children [Delay_2]}
 {id=Looping_4, children [Delay_3]}
 {id=Looping_5, children [Delay_4]}
 {id=Mixer_0, children [WavePlayer_0,Looping_4], properties {inputVolume [3,1.2]}}
 {id=Mixer_1, children [Modulator_0,null,Looping_1,Looping_5], properties {inputVolume [0.8,1.3,0.8,1.3]}}
 {id=Mixer_2, children [Mixer_0,Looping_0], properties {inputVolume [1,0.8]}}
 {id=Modulator_0, children [Looping_3]}
 {id=Modulator_1, children [WavePlayer_17], properties {pitchMax=1, pitchMin=1, volumeMax=0.98}}
 {id=Random_0, children [WavePlayer_1, WavePlayer_13, WavePlayer_14, WavePlayer_15, Modulator_1], properties {weights [1,1,1,1,0.4]}}
 {id=Random_1, children [WavePlayer_2,WavePlayer_3,WavePlayer_4,WavePlayer_5,WavePlayer_6], properties {weights [1,1,1,1,1]}}
 {id=Random_3, children [WavePlayer_9, WavePlayer_10, WavePlayer_11], properties {weights [1,1,1]}}
 {id=Random_4, children [WavePlayer_18, WavePlayer_19, WavePlayer_20, WavePlayer_21, WavePlayer_22], properties {weights [1,1,1,1,1]}}
 {id=WavePlayer_0, properties {bLooping=true, soundWaveAssetPtr=&/Emote_CinderMax_Fire_LP}}
 {id=WavePlayer_1, properties {soundWaveAssetPtr=&/Emote_CinderMax_Alarm_LP}}
 {id=WavePlayer_10, properties {soundWaveAssetPtr=&/Emote_CinderMax_Sirens_02}}
 {id=WavePlayer_11, properties {soundWaveAssetPtr=&/Emote_CinderMax_Sirens_03}}
 {id=WavePlayer_13, properties {soundWaveAssetPtr=&/Emote_CinderMax_Alarm_LP_06}}
 {id=WavePlayer_14, properties {soundWaveAssetPtr=&/Emote_CinderMax_Alarm_LP_02}}
 {id=WavePlayer_15, properties {soundWaveAssetPtr=&/Emote_CinderMax_Alarm_LP_03}}
 {id=WavePlayer_17, properties {soundWaveAssetPtr=&/Emote_CinderMax_Alarm_LP_05}}
 {id=WavePlayer_18, properties {soundWaveAssetPtr=&/Emote_CinderMax_CarCrash_03}}
 {id=WavePlayer_19, properties {soundWaveAssetPtr=&/Emote_CinderMax_CarCrash_01}}
 {id=WavePlayer_2, properties {soundWaveAssetPtr=&/Emote_CinderMax_Honks_01}}
 {id=WavePlayer_20, properties {soundWaveAssetPtr=&/Emote_CinderMax_CarCrash_02}}
 {id=WavePlayer_21, properties {soundWaveAssetPtr=&/Emote_CinderMax_CarCrash_04}}
 {id=WavePlayer_22, properties {soundWaveAssetPtr=&/Emote_CinderMax_CarCrash_05}}
 {id=WavePlayer_3, properties {soundWaveAssetPtr=&/Emote_CinderMax_Honks_02}}
 {id=WavePlayer_4, properties {soundWaveAssetPtr=&/Emote_CinderMax_Honks_03}}
 {id=WavePlayer_5, properties {soundWaveAssetPtr=&/Emote_CinderMax_Honks_04}}
 {id=WavePlayer_6, properties {soundWaveAssetPtr=&/Emote_CinderMax_Honks_05}}
 {id=WavePlayer_9, properties {soundWaveAssetPtr=&/Emote_CinderMax_Sirens_01}}
]
properties {
 attenuationSettings=&Fort_Emotes_Athena_Att
 audioPropertiesSheet=/BRCosmetics/Sounds/Settings/PropertySheets/Emotes/APS_Cosmetic_Emote_Foley
 maxDistance=1600
 soundClassObject=/Game/Sounds/FortSoundClassesAndModes/Fort_Foley_Emote
 virtualizationMode=PlayWhenSilent
 volumeMultiplier=0.4
 concurrencyOverrides {maxCount=3}
 concurrencySet [/Game/Sounds/Concurrency_Rules/Athena_EmoteFoley_Conc]
 soundSubmixSends [{minSendLevel=0, soundSubmix=&Fortnite_AudioAnalysis_Music}]
}
